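; Parent Blueprint for the Creative Hiding Prop device, which players hide inside and can use for hidden travel to grouped props.
; Enabled-state and team/class restriction components gate interaction and falling in, ejecting disqualified or disabled occupants.
; Hiding broadcasts gameplay events, runs a queue-based max-hide-time eject timer and a re-entry block timer, and picks a travel target from an indexed object register.
; A Verse override struct, set from user options, supplies these settings.
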(event OnPropertiesLoaded(ToyOptionsComponent)
  (bind _fortminigameprogress (Variables|Default|GetFortMinigameProgress))
  (bind _currentminigame (Variables|/Script/FortniteGame.FortMinigameLogicComponent|GetCurrentMinigame _fortminigameprogress))
  ;[1 1 1] Initialization
  (CallFunction|SetVerseOverridestoUserOptions)
  (Utilities|FlowControl|Sequence
    (:then_0
      ;[1 1 1] Initialize
      (Class|CreativeEnabledComponent|Initialize (Variables|Default|GetCreative_Enabled_Component) (Variables|Default|GetFortMinigameProgress) (select (Variables|UserOptions|Basic|GetEnabledAtGameStart) "NewEnumerator4" "NewEnumerator1"))
      (Initialization|Initialize (Variables|Default|GetCreative_UseRestriction_Component) (Variables|Default|GetFortMinigameProgress) (Variables|Default|GetCreative_Enabled_Component) (Variables|Default|GetToyOptionsComponent) (Variables|UserOptions|Basic|GetUsableByTeam%20) (Variables|UserOptions|Basic|GetbInvertTeamSelection) (Variables|UserOptions|Basic|GetUsableByClass) (Variables|UserOptions|Basic|GetbInvertClassSelection) false)
      (bind _hiddentravelgroup_override (Verse|GetHiddenTravelGroupOverrideVerse))
      (bind _hiddentraveltargetgroup_override (Verse|GetHiddenTravelTargetGroupOverrideVerse))
      (CallFunction|SetInvulnerability (Math|Boolean|ORBoolean (Variables|UserOptions|Basic|GetInvulnerable%20) _hiddentravelgroup_override _hiddentraveltargetgroup_override))
      (bind _wobblewhilehiding (Verse|GetWobbleWhileHidingVerse))
      (Variables|Default|SetbCanRustleAndWobble _wobblewhilehiding))
    (:then_1
      ;[1 1 1] Keep track of when player Team and Class change to eject players if they are inside and get disqualified
      (Networking|SwitchHasAuthority
        (:Authority
          (if (Variables|UserOptions|GetEjectWhenFailingRequirements%20)
            ;;[1 1 1] Bind Events For Team And Class Changes
            (Utilities|IsValid _currentminigame
              (:Is%20Valid
                (Minigame|AssignOnPlayerClassApplied _currentminigame (delegate Custom|OnPlayerClassApplied_Event))
                (Minigame|BindEventtoPlayerChangedTeamDelegate _currentminigame (delegate Custom|PlayerChangedTeamDel))
                (bind _returnvalue (GameplayEventRouter|GetEventRouter "/Script/FortniteGame.FortPlayspace" self)) ; ContextActor is how we bind to a specific playspace
                (bind (_asynctaskproxy _payload) (GameplayEventRouter|StartListeningToEvent _returnvalue "/Script/FortniteGame.FortTeamChangedContext")
                  (:then
                    (Variables|Default|SetTeamChangedObserver _asynctaskproxy))
                  (:OnEventReceived
                    (bind (_payload_player _payload_teamid) (Utilities|Struct|BreakFortTeamChangedContext _payload))
                    (CallFunction|PlayerChangedTeamDel _payload_player _payload_teamid 0 true))) ; Context here is irrelevant
              ))
            (else
              (Utilities|IsValid _currentminigame
                (:Is%20Valid
                  (Minigame|UnbindEventfromOnPlayerClassApplied _currentminigame (delegate Custom|OnPlayerClassApplied_Event))
                  (Minigame|UnbindEventfromPlayerChangedTeamDelegate _currentminigame (delegate Custom|PlayerChangedTeamDel))
                  (GameplayEventRouter|StopListeningtoEvent (Variables|Default|GetTeamChangedObserver)))))))))
    (:then_2
      ;[1 1 1] Setup Teleporting Group
      (Creative|Networking|Authority/Cosmetic
        (:Authority
          (bind _hiddentravelgroup_override_2 (Verse|GetHiddenTravelGroupOverrideVerse))
          (if (not _hiddentravelgroup_override_2)
            (Variables|UserOptions|Advanced|SetHiddenTravelGroup 0))
          (if (!= (Variables|Vars|GetPreviousTeleportGroup) (Variables|UserOptions|Advanced|GetHiddenTravelGroup))
            (bind _hiddentravelgroup (Verse|GetHiddenTravelGroupVerse))
            (CreativeObjectRegistry|CreateRegisterandUpdateActorIndex (Variables|Default|GetTeleportGroupObjectRegisterComp) "(TagName=\"Creative.Register.HidingProp.TeleportGroup\")" self (Variables|Vars|GetPreviousTeleportGroup) _hiddentravelgroup)))))))

(event OnAnyPropertyChanged(ToyOptionsComponent)
  (bind _fortminigameprogress (Variables|Default|GetFortMinigameProgress))
  (bind _currentminigame (Variables|/Script/FortniteGame.FortMinigameLogicComponent|GetCurrentMinigame _fortminigameprogress))
  ;[1 1 1] Initialization
  (CallFunction|SetVerseOverridestoUserOptions)
  (Utilities|FlowControl|Sequence
    (:then_0
      ;[1 1 1] Initialize
      (Class|CreativeEnabledComponent|Initialize (Variables|Default|GetCreative_Enabled_Component) (Variables|Default|GetFortMinigameProgress) (select (Variables|UserOptions|Basic|GetEnabledAtGameStart) "NewEnumerator4" "NewEnumerator1"))
      (Initialization|Initialize (Variables|Default|GetCreative_UseRestriction_Component) (Variables|Default|GetFortMinigameProgress) (Variables|Default|GetCreative_Enabled_Component) (Variables|Default|GetToyOptionsComponent) (Variables|UserOptions|Basic|GetUsableByTeam%20) (Variables|UserOptions|Basic|GetbInvertTeamSelection) (Variables|UserOptions|Basic|GetUsableByClass) (Variables|UserOptions|Basic|GetbInvertClassSelection) false)
      (bind _hiddentravelgroup_override (Verse|GetHiddenTravelGroupOverrideVerse))
      (bind _hiddentraveltargetgroup_override (Verse|GetHiddenTravelTargetGroupOverrideVerse))
      (CallFunction|SetInvulnerability (Math|Boolean|ORBoolean (Variables|UserOptions|Basic|GetInvulnerable%20) _hiddentravelgroup_override _hiddentraveltargetgroup_override))
      (bind _wobblewhilehiding (Verse|GetWobbleWhileHidingVerse))
      (Variables|Default|SetbCanRustleAndWobble _wobblewhilehiding))
    (:then_1
      ;[1 1 1] Keep track of when player Team and Class change to eject players if they are inside and get disqualified
      (Networking|SwitchHasAuthority
        (:Authority
          (if (Variables|UserOptions|GetEjectWhenFailingRequirements%20)
            ;;[1 1 1] Bind Events For Team And Class Changes
            (Utilities|IsValid _currentminigame
              (:Is%20Valid
                (Minigame|AssignOnPlayerClassApplied _currentminigame (delegate Custom|OnPlayerClassApplied_Event))
                (Minigame|BindEventtoPlayerChangedTeamDelegate _currentminigame (delegate Custom|PlayerChangedTeamDel))
                (bind _returnvalue (GameplayEventRouter|GetEventRouter "/Script/FortniteGame.FortPlayspace" self)) ; ContextActor is how we bind to a specific playspace
                (bind (_asynctaskproxy _payload) (GameplayEventRouter|StartListeningToEvent _returnvalue "/Script/FortniteGame.FortTeamChangedContext")
                  (:then
                    (Variables|Default|SetTeamChangedObserver _asynctaskproxy))
                  (:OnEventReceived
                    (bind (_payload_player _payload_teamid) (Utilities|Struct|BreakFortTeamChangedContext _payload))
                    (CallFunction|PlayerChangedTeamDel _payload_player _payload_teamid 0 true))) ; Context here is irrelevant
              ))
            (else
              (Utilities|IsValid _currentminigame
                (:Is%20Valid
                  (Minigame|UnbindEventfromOnPlayerClassApplied _currentminigame (delegate Custom|OnPlayerClassApplied_Event))
                  (Minigame|UnbindEventfromPlayerChangedTeamDelegate _currentminigame (delegate Custom|PlayerChangedTeamDel))
                  (GameplayEventRouter|StopListeningtoEvent (Variables|Default|GetTeamChangedObserver)))))))))
    (:then_2
      ;[1 1 1] Setup Teleporting Group
      (Creative|Networking|Authority/Cosmetic
        (:Authority
          (bind _hiddentravelgroup_override_2 (Verse|GetHiddenTravelGroupOverrideVerse))
          (if (not _hiddentravelgroup_override_2)
            (Variables|UserOptions|Advanced|SetHiddenTravelGroup 0))
          (if (!= (Variables|Vars|GetPreviousTeleportGroup) (Variables|UserOptions|Advanced|GetHiddenTravelGroup))
            (bind _hiddentravelgroup (Verse|GetHiddenTravelGroupVerse))
            (CreativeObjectRegistry|CreateRegisterandUpdateActorIndex (Variables|Default|GetTeleportGroupObjectRegisterComp) "(TagName=\"Creative.Register.HidingProp.TeleportGroup\")" self (Variables|Vars|GetPreviousTeleportGroup) _hiddentravelgroup)))))))

(event OnAddedtoMinigame(FortMinigameProgress)
  (params (Minigame :type Fort%20Minigame%20Object%20Reference))
  (bind _fortminigameprogress (Variables|Default|GetFortMinigameProgress))
  (bind _currentminigame (Variables|/Script/FortniteGame.FortMinigameLogicComponent|GetCurrentMinigame _fortminigameprogress))
  ;[1 1 1] Keep track of when player Team and Class change to eject players if they are inside and get disqualified
  (Minigame|BindEventtoOnPlayModeChangedDelegate Minigame (delegate Custom|PlayModeChanged))
  (Networking|SwitchHasAuthority
    (:Authority
      (if (Variables|UserOptions|GetEjectWhenFailingRequirements%20)
        ;;[1 1 1] Bind Events For Team And Class Changes
        (Utilities|IsValid _currentminigame
          (:Is%20Valid
            (Minigame|AssignOnPlayerClassApplied _currentminigame (delegate Custom|OnPlayerClassApplied_Event))
            (Minigame|BindEventtoPlayerChangedTeamDelegate _currentminigame (delegate Custom|PlayerChangedTeamDel))
            (bind _returnvalue (GameplayEventRouter|GetEventRouter "/Script/FortniteGame.FortPlayspace" self)) ; ContextActor is how we bind to a specific playspace
            (bind (_asynctaskproxy _payload) (GameplayEventRouter|StartListeningToEvent _returnvalue "/Script/FortniteGame.FortTeamChangedContext")
              (:then
                (Variables|Default|SetTeamChangedObserver _asynctaskproxy))
              (:OnEventReceived
                (bind (_payload_player _payload_teamid) (Utilities|Struct|BreakFortTeamChangedContext _payload))
                (CallFunction|PlayerChangedTeamDel _payload_player _payload_teamid 0 true))) ; Context here is irrelevant
          ))
        (else
          (Utilities|IsValid _currentminigame
            (:Is%20Valid
              (Minigame|UnbindEventfromOnPlayerClassApplied _currentminigame (delegate Custom|OnPlayerClassApplied_Event))
              (Minigame|UnbindEventfromPlayerChangedTeamDelegate _currentminigame (delegate Custom|PlayerChangedTeamDel))
              (GameplayEventRouter|StopListeningtoEvent (Variables|Default|GetTeamChangedObserver)))))))))

(event Custom|PlayModeChanged
  (params (Minigame :type Fort%20Minigame%20Object%20Reference) (bIsInPlayMode :type Boolean))
  ;[1 1 1] Keep track of when player Team and Class change to eject players if they are inside and get disqualified
  (CallFunction|ResetDevice))

(event Custom|PlayerChangedTeamDel
  (params (PlayerState :type Fort%20Player%20State%20Object%20Reference) (TeamId :type Byte) (OldTeamId :type Byte) (bTeamChangedWithoutRespawn :type Boolean))
  (Variables|Vars|SetPlayer%20State PlayerState)
  (if (Utilities|Array|ContainsItem (Variables|Default|GetHidingPlayers) (PlayerState|GetCurrentPawn (Variables|Vars|GetPlayer%20State)))
    (bind (_valid_controller _fortplayerpawn) (CreativeUserRestrictions|ActorHasValidParticipantController (Variables|Default|GetCreative_UseRestriction_Component) (Variables|Vars|GetPlayer%20State) true false true))
    (if (not _valid_controller)
      (CallFunction|StopHiding _fortplayerpawn))))

(event Custom|OnPlayerClassApplied_Event
  (params (PlayerState :type Fort%20Player%20State%20Object%20Reference) (ClassSlotIndex :type Byte))
  (Variables|Vars|SetPlayer%20State PlayerState)
  (if (Utilities|Array|ContainsItem (Variables|Default|GetHidingPlayers) (PlayerState|GetCurrentPawn (Variables|Vars|GetPlayer%20State)))
    (bind (_valid_controller _fortplayerpawn) (CreativeUserRestrictions|ActorHasValidParticipantController (Variables|Default|GetCreative_UseRestriction_Component) (Variables|Vars|GetPlayer%20State) true false true))
    (if (not _valid_controller)
      (CallFunction|StopHiding _fortplayerpawn))))

(event Custom|Enable%20Event
  (params (Instigator :type Controller%20Object%20Reference))
  ;[1 1 1] Gameplay Functions
  (Class|CreativeEnabledComponent|SetEnabled (Variables|Default|GetCreative_Enabled_Component) true))

(event EventOnDeathServer
  (params (Damage :type Float%20%28single-precision%29) (DamageTags :type Gameplay%20Tag%20Container%20Structure%20%28by%20ref%29) (Momentum :type Vector :default "0, 0, 0") (HitInfo :type Hit%20Result%20Structure%20%28by%20ref%29) (InstigatedBy :type Controller%20Object%20Reference) (DamageCauser :type Actor%20Object%20Reference) (EffectContext :type Gameplay%20Effect%20Context%20Handle%20Structure))
  ;[1 1 1] Gameplay Events
  (|CallParentFunction Damage DamageTags Momentum HitInfo InstigatedBy DamageCauser EffectContext)
  (GameplayEvent|BroadcastEvent (Variables|UserOptions|Events|GetOnPropDestroyed) 0 (Utilities|Struct|MakeGameplayEventDestroyed InstigatedBy))
  (CallFunction|DisableHiddenTravelEvent 0)
  (CallFunction|EjectAll))

(event OnEnabledStateChanged(Creative_Enabled_Component)
  (params (Enabled :type Boolean))
  ;[1 1 1] Gameplay Functions
  (Networking|SwitchHasAuthority
    (:Authority
      (if (not Enabled)
        (CallFunction|EjectAll)))))

(event Custom|Disable%20Event
  (params (Instigator :type Controller%20Object%20Reference))
  ;[1 1 1] Gameplay Functions
  (Class|CreativeEnabledComponent|SetEnabled (Variables|Default|GetCreative_Enabled_Component) false))

(event Custom|ToggleInvulnerability%20Event
  ;[1 1 1] Gameplay Functions
  (CallFunction|SetInvulnerability (Variables|Actor|GetbCanBeDamaged)))

(event Custom|EjectAllHidingPlayers%20Event
  ;[1 1 1] Gameplay Functions
  (CallFunction|EjectAll))

(event Custom|EjectHidingPlayer%20Event
  (params (Instigator :type Controller%20Object%20Reference))
  ;[1 1 1] Gameplay Functions
  (bind _asfort_pawn (Utilities|Casting|CastToFortPawn (Pawn|GetControlledPawn Instigator)))
  (if (or (CallFunction|PlayerHiddenInProp _asfort_pawn) (CallFunction|PlayerGettingInProp _asfort_pawn))
    (CallFunction|StopHiding _asfort_pawn)))

(event Custom|HideNearbyPlayers%20Event
  ;[1 1 1] Gameplay Functions
  (CallFunction|HideNearyPlayersInternal 1e+03))

(event Custom|DestroyProp
  (params (Instigator :type Controller%20Object%20Reference))
  ;[1 1 1] Gameplay Functions
  (Game|Damage|ApplyDamage self (Health|GetHealth) Instigator Instigator))

(event Custom|UpdateTeleportTarget
  (params (NewTeleportTarget :type B%20Hiding%20Prop%20Object%20Reference))
  ;[1 1 1] Teleporting
  (Variables|Default|SetTargetTeleporter NewTeleportTarget)
  (Networking|ForceNetUpdate))

(event Custom|EnableHiddenTravel%20Event
  (params (NewParam :type Controller%20Object%20Reference))
  ;[1 1 1] Teleporting
  (Variables|Vars|SetHiddenTravelEnabledOnDevice true)
  (Utilities|Struct|SetmembersinHidingPropVerseOverrideValues (Variables|Verse|GetVerseOverrides) true)
  (Utilities|FlowControl|Sequence
    (:then_0
      (if (not (Variables|UserOptions|Basic|GetInvulnerable%20))
        (CallFunction|SetInvulnerability true)))
    (:then_1
      (bind _hiddentravelgroup_override (Verse|GetHiddenTravelGroupOverrideVerse))
      (if _hiddentravelgroup_override
        (bind _hiddentravelgroup (Verse|GetHiddenTravelGroupVerse))
        (CreativeObjectRegistry|CreateRegisterandUpdateActorIndex (Variables|Default|GetTeleportGroupObjectRegisterComp) "(TagName=\"Creative.Register.HidingProp.TeleportGroup\")" self 0 _hiddentravelgroup)))))

(event Custom|DisableHiddenTravel%20Event
  (params (NewParam :type Controller%20Object%20Reference))
  ;[1 1 1] Teleporting
  (Variables|Vars|SetHiddenTravelEnabledOnDevice false)
  (Utilities|Struct|SetmembersinHidingPropVerseOverrideValues (Variables|Verse|GetVerseOverrides) false)
  (CallFunction|UpdateTeleportTarget 0)
  (Utilities|FlowControl|Sequence
    (:then_0
      (if (not (Variables|UserOptions|Basic|GetInvulnerable%20))
        (CallFunction|SetInvulnerability false)))
    (:then_1
      (bind _hiddentravelgroup_override (Verse|GetHiddenTravelGroupOverrideVerse))
      (if _hiddentravelgroup_override
        (bind _hiddentravelgroup (Verse|GetHiddenTravelGroupVerse))
        (CreativeObjectRegistry|CreateRegisterandUpdateActorIndex (Variables|Default|GetTeleportGroupObjectRegisterComp) "(TagName=\"Creative.Register.HidingProp.TeleportGroup\")" self _hiddentravelgroup)))))

(event MinigameStateChangedDelegate(FortMinigameProgress)
  (params (Minigame :type Fort%20Minigame%20Object%20Reference) (NewMinigameState :type EFortMinigameState%20Enum :default "PreGame"))
  (switch Utilities|FlowControl|Switch|SwitchonEFortMinigameState NewMinigameState
    (:Setup
      _)
    (:Transitioning
      _)
    (:PostRoundEnd
      _)
    (:PostGameEnd
      _))
  (CallFunction|SetVerseOverridestoUserOptions))

(event EventHiddenTravelCompleteEvent
  (params (TargetHidingProp :type B%20Hiding%20Prop%20Object%20Reference) (Pawn :type Fort%20Pawn%20Object%20Reference))
  (|CallParentFunction 0 Pawn)
  (bind _asdevice_hiding_prop_parent (Utilities|Casting|CastToDevice_HidingProp_Parent TargetHidingProp))
  (GameplayEvent|BroadcastEvent (Variables|UserOptions|Events|GetOnHiddenTravelComplete) 0 (Utilities|Struct|MakeGameplayEventonHiddenTravelComplete (Pawn|GetController Pawn) _asdevice_hiding_prop_parent)))

(fn AddHiddenPlayer
  (params (FortPawn :type Fort%20Pawn%20Object%20Reference))
  ;[1 1 1] Pick the target for next time
  (CallFunction|ServerPickTeleportTarget FortPawn)
  (|CallParentFunction (Variables|GetFortPawn))
  ;[1 1 1] On Hide Event
  (GameplayEvent|BroadcastEvent (Variables|UserOptions|Events|GetOnHide) 0 (Utilities|Struct|MakeGameplayEventonHide (Pawn|GetController (Variables|GetFortPawn)) self))
  (Utilities|FlowControl|Sequence
    (:then_0
      ;[1 1 1] If a maximum hiding time has been set (and Not Teleporting) then queue up the player and ensure a timer is running to eject the next player
      (bind _maxhidingtime_override (Verse|GetMaxHidingTimeOverrideVerse))
      (bind _hiddentraveltargetgroup_override (Verse|GetHiddenTravelTargetGroupOverrideVerse))
      (if (and _maxhidingtime_override (not _hiddentraveltargetgroup_override))
        (Utilities|Array|Add (Variables|Vars|GetEjectPlayerAfterTimeQueue_EntryTime) (Actor|GetGameTimeSinceCreation))
        (Utilities|Array|Add (Variables|Vars|GetEjectPlayerAfterTimeQueue_Pawn) (Variables|GetFortPawn))
        (if (== (Utilities|Array|Length (Variables|Vars|GetEjectPlayerAfterTimeQueue_Pawn)) 1)
          (bind _maxhidingtime (Verse|GetMaxHidingTimeVerse))
          (bind _returnvalue (Utilities|Time|SetTimerbyEvent (EventDispatchers|CreateEvent :!event EjectPlayerAfterHideTime) _maxhidingtime))
          (Variables|Vars|SetEjectPlayerAfterHideTime_TimerHandle _returnvalue))))
    (:then_1
      ;[1 1 1] Block others from hiding if there's a Block Hide Time
      (bind _blockhidingtime (Verse|GetBlockHideTimeVerse))
      (if (> _blockhidingtime 0)
        (bind _returnvalue_2 (Utilities|Time|SetTimerbyEvent (EventDispatchers|CreateEvent :!event ClearBlockHideTimer) _blockhidingtime))
        (Variables|Vars|SetHideBlockedDuringTimerHandle _returnvalue_2)
        (Variables|Vars|SetbHideBlockedDuringTimer true)))))

(fn AllowInverseScaling
  (return false))

(fn BlueprintCanInteract
  (params (InteractingPawn :type Fort%20Pawn%20Object%20Reference) (InteractionBeingAttempted :type EInteractionBeingAttempted%20Enum :default "FirstInteraction") (InteractionType :type TInteractionType%20Enum :default "IT_NoInteraction"))
  (if (Class|CreativeCommonDeviceFunctionLibrary|IsLogicinPlayMode (Variables|Default|GetFortMinigameProgress))
    (if (CallFunction|PlayerHiddenInProp (Variables|GetInteractingPawn))
      ;[1 1 1] Always let the parent handle making the prompt available when the player is inside so they don't get stuck
      (return (|CallParentFunction (Variables|GetInteractingPawn) (Variables|GetInteractionBeingAttempted) (Variables|GetInteractionType)))
      (else
        ;[1 1 1] During Play we can decide if the interaction is available for this player qualifies
        (bind _valid_controller (CreativeUserRestrictions|ActorHasValidParticipantController (Variables|Default|GetCreative_UseRestriction_Component) (Variables|GetInteractingPawn) true false true))
        (if (and _valid_controller (Class|CreativeEnabledComponent|IsEnabled (Variables|Default|GetCreative_Enabled_Component)))
          (if (not (Variables|Vars|GetbHideBlockedDuringTimer))
            (return (|CallParentFunction (Variables|GetInteractingPawn) (Variables|GetInteractionBeingAttempted) (Variables|GetInteractionType)))
            (else
              _))
          (else
            _))
        (return false)))
    (else
      ;[1 1 1] Should the player be shown the interaction in edit mode?
      (return (Interaction|CanEditOptionsComponent (Variables|Default|GetToyOptionsComponent) (Variables|GetInteractingPawn))))))

(fn BlueprintGetFailedInteractionString
  (params (InteractingPawn :type Fort%20Pawn%20Object%20Reference) (InteractionBeingAttempted :type EInteractionBeingAttempted%20Enum :default "FirstInteraction"))
  (bind _valid_controller (CreativeUserRestrictions|ActorHasValidParticipantController (Variables|Default|GetCreative_UseRestriction_Component) (Variables|GetInteractingPawn) true false true))
  (if (Math|Boolean|ANDBoolean _valid_controller (Class|CreativeEnabledComponent|IsEnabled (Variables|Default|GetCreative_Enabled_Component)) (Class|CreativeCommonDeviceFunctionLibrary|IsLogicinPlayMode (Variables|Default|GetFortMinigameProgress)) (not (Utilities|Text|TextIsEmpty (Variables|UserOptions|GetOccupiedText))) (not (CallFunction|PlayerHiddenInProp (Variables|GetInteractingPawn))) (not (CallFunction|PlayerGettingInProp (Variables|GetInteractingPawn))))
    (return (Variables|UserOptions|GetOccupiedText))
    (else
      (return ""))))

(fn BlueprintGetInteractionString
  (params (InteractingPawn :type Fort%20Pawn%20Object%20Reference) (InteractionBeingAttempted :type EInteractionBeingAttempted%20Enum :default "FirstInteraction"))
  (if (Class|CreativeCommonDeviceFunctionLibrary|IsLogicinPlayMode (Variables|Default|GetFortMinigameProgress))
    (if (Utilities|Array|ContainsItem (Variables|Default|GetHidingPlayers) (Variables|GetInteractingPawn))
      ;[1 1 1] Only called once so need to grab this new string as soon as the player starts getting in.
      (bind _returnvalue (|CallParentFunction (Variables|GetInteractingPawn) (Variables|GetInteractionBeingAttempted)))
      (return (select (Utilities|Text|TextIsEmpty (Variables|UserOptions|GetStopHidingText)) _returnvalue (Variables|UserOptions|GetStopHidingText)))
      (else
        (bind _returnvalue_2 (|CallParentFunction (Variables|GetInteractingPawn) (Variables|GetInteractionBeingAttempted)))
        (return (select (Utilities|Text|TextIsEmpty (Variables|UserOptions|GetHideText)) _returnvalue_2 (Variables|UserOptions|GetHideText)))))
    (else
      (return (Variables|/Script/FortniteGame.FortActorOptionsComponent|GetInternal_InteractionText (Variables|Default|GetToyOptionsComponent))))))

(fn BlueprintGetInteractionTime
  (params (InteractingPawn :type Fort%20Pawn%20Object%20Reference) (InteractionBeingAttempted :type EInteractionBeingAttempted%20Enum :default "FirstInteraction"))
  (if (Class|CreativeCommonDeviceFunctionLibrary|IsLogicinPlayMode (Variables|Default|GetFortMinigameProgress))
    (if (Math|Boolean|ANDBoolean (!= (Variables|UserOptions|GetInteractTime) 0) (not (CallFunction|PlayerHiddenInProp (Variables|GetInteractingPawn))) (not (CallFunction|PlayerGettingInProp (Variables|GetInteractingPawn))))
      (return (Variables|UserOptions|GetInteractTime) true)
      (else
        (bind (_outinteractiontime _returnvalue) (|CallParentFunction (Variables|GetInteractingPawn) (Variables|GetInteractionBeingAttempted)))
        (return _outinteractiontime _returnvalue)))
    (else
      (return 0.0 false))))

(fn BlueprintOnLocalInteract
  (params (InteractingPawn :type Fort%20Player%20Pawn%20Object%20Reference) (InteractionBeingAttempted :type EInteractionBeingAttempted%20Enum :default "FirstInteraction"))
  (if (Class|CreativeCommonDeviceFunctionLibrary|IsLogicinPlayMode (Variables|Default|GetFortMinigameProgress))
    (bind _returnvalue (|CallParentFunction (Variables|GetInteractingPawn) (Variables|GetInteractionBeingAttempted)))
    (return _returnvalue)
    (else
      ;[1 1 1] The ToyOptionsComponent will decied to consume the interact
      (bind _returnvalue_2 (Interaction|OnLocalInteract (Variables|Default|GetToyOptionsComponent) (Variables|GetInteractingPawn)))
      (return _returnvalue_2))))

(fn CanHidebyFalling
  (params (OtherActor :type Actor%20Object%20Reference))
  (bind _valid_controller (CreativeUserRestrictions|ActorHasValidParticipantController (Variables|Default|GetCreative_UseRestriction_Component) OtherActor true false true))
  (if (Math|Boolean|ANDBoolean _valid_controller (Class|CreativeCommonDeviceFunctionLibrary|IsLogicinPlayMode (Variables|Default|GetFortMinigameProgress)) (Class|CreativeEnabledComponent|IsEnabled (Variables|Default|GetCreative_Enabled_Component)) (not (Variables|Vars|GetbHideBlockedDuringTimer)))
    (bind (_canhide _afortplayerpawn) (|CallParentFunction (Variables|GetOtherActor)))
    (return _canhide _afortplayerpawn)
    (else
      (return false 0))))

(fn ClearBlockHideTimer
  (Variables|Vars|SetbHideBlockedDuringTimer false))

(fn EjectAll
  (bind _output_get (LocalVariables|Setlocal_TempHidingPlayers (Variables|Default|GetHidingPlayers)))
  (for _array_element _output_get
    (Utilities|IsValid _array_element
      (:Is%20Valid
        (CallFunction|StopHiding _array_element)
        (Ability|SendGameplayEventtoActor _array_element "(TagName=\"Event.Athena.HidingProp.AbortTeleporting\")" (Utilities|Struct|MakeGameplayEventData "(TagName=\"\")" self 0 self))))))

(fn EjectAllHidingPlayers_Verse
  (params (UnusedFloat :type Float%20%28double-precision%29))
  (bind _output_get (LocalVariables|SetEjectedPlayerPawns (Variables|Default|GetHidingPlayers)))
  (CallFunction|EjectAllHidingPlayersEvent)
  (for _array_element _output_get
    (Utilities|IsValid (Pawn|GetController _array_element)
      (:Is%20Valid
        (Utilities|Array|Add (LocalVariables|GetEjectedPlayerControllers) (Pawn|GetController _array_element)))))
  (return (LocalVariables|GetEjectedPlayerControllers)))

(fn EjectHidingPlayer_Verse
  (params (Instigator :type Controller%20Object%20Reference))
  (bind _asfort_pawn (Utilities|Casting|CastToFortPawn (Pawn|GetControlledPawn Instigator)))
  (if (or (CallFunction|PlayerHiddenInProp _asfort_pawn) (CallFunction|PlayerGettingInProp _asfort_pawn))
    (CallFunction|EjectHidingPlayerEvent Instigator)
    (return true)
    (else
      (return false))))

(fn EjectPlayerAfterHideTime
  (CallFunction|StopHiding (Utilities|Array|Get(acopy) (Variables|Vars|GetEjectPlayerAfterTimeQueue_Pawn))))

(fn ForceGridAlignment
  (return false))

(fn GetAdditionalSelectableActors
  (return 0))

(fn GetAttemptNoRepeatDestinations_Verse
  (bind (_maxnumberofoccupants_27_a675ad9a4b2b4300c05fad9f658daaf9 _wobblewhilehiding_5_f04b5a4f47ee79ebdf664199fa0fd9e2 _maxhidingtime_8_1311547a4ff0bfc0a632208449cd0607 _blockhidingtime_10_c7528f20479c4e311bc8c18955433ae2 _hiddentravelgroup_13_815cc31b4639f9f1c8349b953f15b7d3 _hiddentraveltargetgroup_15_21e65df4424aafac113a92a3b05f5744 _attemptnorepeatdestinations_18_d56708e74ff9314b2928f482af43c896) (Utilities|Struct|BreakHidingPropVerseOverrideValues (Variables|Verse|GetVerseOverrides)))
  (return _attemptnorepeatdestinations_18_d56708e74ff9314b2928f482af43c896))

(fn GetBlockHideTime_Verse
  (bind (_maxnumberofoccupants_27_a675ad9a4b2b4300c05fad9f658daaf9 _wobblewhilehiding_5_f04b5a4f47ee79ebdf664199fa0fd9e2 _maxhidingtime_8_1311547a4ff0bfc0a632208449cd0607 _blockhidingtime_10_c7528f20479c4e311bc8c18955433ae2) (Utilities|Struct|BreakHidingPropVerseOverrideValues (Variables|Verse|GetVerseOverrides)))
  (return _blockhidingtime_10_c7528f20479c4e311bc8c18955433ae2))

(fn GetBlockPitchAndRoll
  (return false))

(fn GetEnableHiddenTravel_Verse
  (bind (_maxnumberofoccupants_27_a675ad9a4b2b4300c05fad9f658daaf9 _wobblewhilehiding_5_f04b5a4f47ee79ebdf664199fa0fd9e2 _maxhidingtime_8_1311547a4ff0bfc0a632208449cd0607 _blockhidingtime_10_c7528f20479c4e311bc8c18955433ae2 _hiddentravelgroup_13_815cc31b4639f9f1c8349b953f15b7d3 _hiddentraveltargetgroup_15_21e65df4424aafac113a92a3b05f5744 _attemptnorepeatdestinations_18_d56708e74ff9314b2928f482af43c896 _enablehiddentravel_20_5e05d3fe46abcbbacfeea28d1f05aac5) (Utilities|Struct|BreakHidingPropVerseOverrideValues (Variables|Verse|GetVerseOverrides)))
  (return _enablehiddentravel_20_5e05d3fe46abcbbacfeea28d1f05aac5))

(fn GetHiddenTravelGroup_Override_Verse
  (bind (_maxnumberofoccupants_27_a675ad9a4b2b4300c05fad9f658daaf9 _wobblewhilehiding_5_f04b5a4f47ee79ebdf664199fa0fd9e2 _maxhidingtime_8_1311547a4ff0bfc0a632208449cd0607 _blockhidingtime_10_c7528f20479c4e311bc8c18955433ae2 _hiddentravelgroup_13_815cc31b4639f9f1c8349b953f15b7d3 _hiddentraveltargetgroup_15_21e65df4424aafac113a92a3b05f5744 _attemptnorepeatdestinations_18_d56708e74ff9314b2928f482af43c896 _enablehiddentravel_20_5e05d3fe46abcbbacfeea28d1f05aac5 _maxhidingtimeoverridebool_22_bb4e333f45fbfbf6d307f197e79ddef2 _hiddentraveloverridebool_24_72f877104200cd29af262baf5415320c) (Utilities|Struct|BreakHidingPropVerseOverrideValues (Variables|Verse|GetVerseOverrides)))
  (return _hiddentraveloverridebool_24_72f877104200cd29af262baf5415320c))

(fn GetHiddenTravelGroup_Verse
  (bind _verseoverrides (Variables|Verse|GetVerseOverrides))
  (bind (_maxnumberofoccupants_27_a675ad9a4b2b4300c05fad9f658daaf9 _wobblewhilehiding_5_f04b5a4f47ee79ebdf664199fa0fd9e2 _maxhidingtime_8_1311547a4ff0bfc0a632208449cd0607 _blockhidingtime_10_c7528f20479c4e311bc8c18955433ae2 _hiddentravelgroup_13_815cc31b4639f9f1c8349b953f15b7d3 _hiddentraveltargetgroup_15_21e65df4424aafac113a92a3b05f5744 _attemptnorepeatdestinations_18_d56708e74ff9314b2928f482af43c896 _enablehiddentravel_20_5e05d3fe46abcbbacfeea28d1f05aac5 _maxhidingtimeoverridebool_22_bb4e333f45fbfbf6d307f197e79ddef2 _hiddentraveloverridebool_24_72f877104200cd29af262baf5415320c) (Utilities|Struct|BreakHidingPropVerseOverrideValues _verseoverrides))
  (if _hiddentraveloverridebool_24_72f877104200cd29af262baf5415320c
    (return _hiddentravelgroup_13_815cc31b4639f9f1c8349b953f15b7d3)
    (else
      (return -1))))

(fn GetHiddenTravelTargetGroup_Override_Verse
  (bind (_maxnumberofoccupants_27_a675ad9a4b2b4300c05fad9f658daaf9 _wobblewhilehiding_5_f04b5a4f47ee79ebdf664199fa0fd9e2 _maxhidingtime_8_1311547a4ff0bfc0a632208449cd0607 _blockhidingtime_10_c7528f20479c4e311bc8c18955433ae2 _hiddentravelgroup_13_815cc31b4639f9f1c8349b953f15b7d3 _hiddentraveltargetgroup_15_21e65df4424aafac113a92a3b05f5744 _attemptnorepeatdestinations_18_d56708e74ff9314b2928f482af43c896 _enablehiddentravel_20_5e05d3fe46abcbbacfeea28d1f05aac5 _maxhidingtimeoverridebool_22_bb4e333f45fbfbf6d307f197e79ddef2 _hiddentraveloverridebool_24_72f877104200cd29af262baf5415320c _hiddentraveltargetoverridebool_26_0b871da6478a35674f4d8ca58363bf24) (Utilities|Struct|BreakHidingPropVerseOverrideValues (Variables|Verse|GetVerseOverrides)))
  (return _hiddentraveltargetoverridebool_26_0b871da6478a35674f4d8ca58363bf24))

(fn GetHiddenTravelTargetGroup_Verse
  (bind _verseoverrides (Variables|Verse|GetVerseOverrides))
  (bind (_maxnumberofoccupants_27_a675ad9a4b2b4300c05fad9f658daaf9 _wobblewhilehiding_5_f04b5a4f47ee79ebdf664199fa0fd9e2 _maxhidingtime_8_1311547a4ff0bfc0a632208449cd0607 _blockhidingtime_10_c7528f20479c4e311bc8c18955433ae2 _hiddentravelgroup_13_815cc31b4639f9f1c8349b953f15b7d3 _hiddentraveltargetgroup_15_21e65df4424aafac113a92a3b05f5744 _attemptnorepeatdestinations_18_d56708e74ff9314b2928f482af43c896 _enablehiddentravel_20_5e05d3fe46abcbbacfeea28d1f05aac5 _maxhidingtimeoverridebool_22_bb4e333f45fbfbf6d307f197e79ddef2 _hiddentraveloverridebool_24_72f877104200cd29af262baf5415320c _hiddentraveltargetoverridebool_26_0b871da6478a35674f4d8ca58363bf24) (Utilities|Struct|BreakHidingPropVerseOverrideValues _verseoverrides))
  (if _hiddentraveltargetoverridebool_26_0b871da6478a35674f4d8ca58363bf24
    (return _hiddentraveltargetgroup_15_21e65df4424aafac113a92a3b05f5744)
    (else
      (return -1))))

(fn GetMaxHidingTime_Override_Verse
  (bind (_maxnumberofoccupants_27_a675ad9a4b2b4300c05fad9f658daaf9 _wobblewhilehiding_5_f04b5a4f47ee79ebdf664199fa0fd9e2 _maxhidingtime_8_1311547a4ff0bfc0a632208449cd0607 _blockhidingtime_10_c7528f20479c4e311bc8c18955433ae2 _hiddentravelgroup_13_815cc31b4639f9f1c8349b953f15b7d3 _hiddentraveltargetgroup_15_21e65df4424aafac113a92a3b05f5744 _attemptnorepeatdestinations_18_d56708e74ff9314b2928f482af43c896 _enablehiddentravel_20_5e05d3fe46abcbbacfeea28d1f05aac5 _maxhidingtimeoverridebool_22_bb4e333f45fbfbf6d307f197e79ddef2) (Utilities|Struct|BreakHidingPropVerseOverrideValues (Variables|Verse|GetVerseOverrides)))
  (return _maxhidingtimeoverridebool_22_bb4e333f45fbfbf6d307f197e79ddef2))

(fn GetMaxHidingTime_Verse
  (bind _verseoverrides (Variables|Verse|GetVerseOverrides))
  (bind (_maxnumberofoccupants_27_a675ad9a4b2b4300c05fad9f658daaf9 _wobblewhilehiding_5_f04b5a4f47ee79ebdf664199fa0fd9e2 _maxhidingtime_8_1311547a4ff0bfc0a632208449cd0607 _blockhidingtime_10_c7528f20479c4e311bc8c18955433ae2 _hiddentravelgroup_13_815cc31b4639f9f1c8349b953f15b7d3 _hiddentraveltargetgroup_15_21e65df4424aafac113a92a3b05f5744 _attemptnorepeatdestinations_18_d56708e74ff9314b2928f482af43c896 _enablehiddentravel_20_5e05d3fe46abcbbacfeea28d1f05aac5 _maxhidingtimeoverridebool_22_bb4e333f45fbfbf6d307f197e79ddef2) (Utilities|Struct|BreakHidingPropVerseOverrideValues _verseoverrides))
  (if _maxhidingtimeoverridebool_22_bb4e333f45fbfbf6d307f197e79ddef2
    (return _maxhidingtime_8_1311547a4ff0bfc0a632208449cd0607)
    (else
      (return -1.0))))

(fn GetNumberOfOccupants_Verse
  (return (Utilities|Struct|BreakHidingPropVerseOverrideValues (Variables|Verse|GetVerseOverrides))))

(fn GetPlayerLimit
  (bind _numberofoccupants (Verse|GetNumberOfOccupantsVerse))
  (return _numberofoccupants))

(fn GetScaleMax
  (params (DefaultScaleMax :type Float%20%28single-precision%29))
  (return 1.0))

(fn GetScaleMin
  (params (DefaultScaleMin :type Float%20%28single-precision%29))
  (return 1.0))

(fn GetSpawnedActors
  (return 0))

(fn GetWobbleWhileHiding_Verse
  (bind (_maxnumberofoccupants_27_a675ad9a4b2b4300c05fad9f658daaf9 _wobblewhilehiding_5_f04b5a4f47ee79ebdf664199fa0fd9e2) (Utilities|Struct|BreakHidingPropVerseOverrideValues (Variables|Verse|GetVerseOverrides)))
  (return _wobblewhilehiding_5_f04b5a4f47ee79ebdf664199fa0fd9e2))

(fn HasAdditionalSelectableActors
  (return false))

(fn HasAdditionalValidPlacementLogic
  (return false))

(fn HideNearbyPlayers_Range
  (params (Range :type Float%20%28double-precision%29))
  (LocalVariables|SetAlreadyHiddenPlayers (Variables|Default|GetHidingPlayers))
  (CallFunction|HideNearyPlayersInternal Range)
  (for _array_element (Variables|Default|GetHidingPlayers)
    (if (not (> (Utilities|Array|FindItem (LocalVariables|GetAlreadyHiddenPlayers) _array_element) -1))
      (Utilities|IsValid (Pawn|GetController _array_element)
        (:Is%20Valid
          (Utilities|Array|AddUnique (LocalVariables|GetNewlyHiddenPlayers) (Pawn|GetController _array_element))))))
  (return (LocalVariables|GetNewlyHiddenPlayers)))

(fn HideNearyPlayers_Internal
  (params (Range :type Float%20%28double-precision%29 :default 1e+03))
  (bind _outfortplayerpawns (Fortnite|GetFortPlayerPawnsInRange (Transformation|GetActorLocation self) Range))
  (for _array_element _outfortplayerpawns
    (if (and (Interaction|BlueprintCanInteract _array_element) (not (GameplayTags|HasMatchingGameplayTag _array_element "(TagName=\"Gameplay.Status.DBNO\")")))
      (CallFunction|StartHiding _array_element))))

(fn IsEnabled
  (return (Class|CreativeEnabledComponent|IsEnabled (Variables|Default|GetCreative_Enabled_Component))))

(fn IsValidPlacement
  (return false))

(fn OnPreviewEnd
  (return false))

(fn OnPreviewStart
  (return false))

(fn RemoveHiddenPlayer
  (params (FortPawn :type Fort%20Pawn%20Object%20Reference))
  (|CallParentFunction FortPawn)
  ;[1 1 1] OnStopHiding
  (GameplayEvent|BroadcastEvent (Variables|UserOptions|Events|GetOnStopHiding) 0 (Utilities|Struct|MakeGameplayEventonStopHiding (Pawn|GetController (Variables|GetFortPawn)) self))
  ;[1 1 1] If a maximum hiding time has been set then we need to remove hiding players from the queue
  (bind _maxhidingtime_override (Verse|GetMaxHidingTimeOverrideVerse))
  (if _maxhidingtime_override
    (LocalVariables|Setlocal_IndexOfPawnInEjectQueue (Utilities|Array|FindItem (Variables|Vars|GetEjectPlayerAfterTimeQueue_Pawn) (Variables|GetFortPawn)))
    (Utilities|Array|RemoveIndex (Variables|Vars|GetEjectPlayerAfterTimeQueue_Pawn) (LocalVariables|Getlocal_IndexOfPawnInEjectQueue))
    (Utilities|Array|RemoveIndex (Variables|Vars|GetEjectPlayerAfterTimeQueue_EntryTime) (LocalVariables|Getlocal_IndexOfPawnInEjectQueue))
    ;[1 1 1] If the player that was removed was the first player in the queue and there are more players to eject then update the timer.  If not then this player jumped out early and there never was a timer for this player, or this is the last player.
    (if (and (== (LocalVariables|Getlocal_IndexOfPawnInEjectQueue) 0) (> (Utilities|Array|Length (Variables|Vars|GetEjectPlayerAfterTimeQueue_Pawn)) 0))
      (Utilities|Time|ClearandInvalidateTimerbyHandle (Variables|Vars|GetEjectPlayerAfterHideTime_TimerHandle))
      (bind _maxhidingtime (Verse|GetMaxHidingTimeVerse))
      (bind _returnvalue (Utilities|Time|SetTimerbyEvent (EventDispatchers|CreateEvent :!event EjectPlayerAfterHideTime) (- _maxhidingtime (- (Actor|GetGameTimeSinceCreation) (Utilities|Array|Get(aref) (Variables|Vars|GetEjectPlayerAfterTimeQueue_EntryTime))))))
      (Variables|Vars|SetEjectPlayerAfterHideTime_TimerHandle _returnvalue))))

(fn ResetDevice
  (Utilities|Array|Clear (Variables|Vars|GetEjectPlayerAfterTimeQueue_EntryTime))
  (Utilities|Array|Clear (Variables|Vars|GetEjectPlayerAfterTimeQueue_Pawn))
  (Utilities|Time|ClearandInvalidateTimerbyHandle (Variables|Vars|GetEjectPlayerAfterHideTime_TimerHandle))
  (Variables|Vars|SetbHideBlockedDuringTimer false)
  (Utilities|Time|ClearandInvalidateTimerbyHandle (Variables|Vars|GetHideBlockedDuringTimerHandle))
  (CallFunction|EjectAll)
  (Variables|Vars|SetPreviousTeleportTarget 0)
  (Creative|Networking|Authority/Cosmetic
    (:Authority
      (CallFunction|UpdateTeleportTarget 0))))

(fn ServerPickTeleportTarget
  (params (NewParam :type Fort%20Pawn%20Object%20Reference))
  (Creative|Networking|Authority/Cosmetic
    (:Authority
      (bind _enablehiddentravel (Verse|GetEnableHiddenTravelVerse))
      (bind _hiddentraveltargetgroup_override (Verse|GetHiddenTravelTargetGroupOverrideVerse))
      (if (Math|Boolean|ANDBoolean _hiddentraveltargetgroup_override _enablehiddentravel (not (GameplayTags|HasMatchingGameplayTag NewParam "(TagName=\"Granted.Athena.EnvItem.HidingProp.Teleporter.Teleporting\")")))
        (bind _teleportgroupobjectregistercomp (Variables|Default|GetTeleportGroupObjectRegisterComp)
          (:then
            _)
          (:else
            (CreativeObjectRegistry|GetOrCreateObjectRegister self "(TagName=\"Creative.Register.HidingProp.TeleportGroup\")" false 0 _teleportgroupobjectregistercomp)))
        (bind _hiddentraveltargetgroup (Verse|GetHiddenTravelTargetGroupVerse))
        (bind (_bisvalid _register) (Class|FortCreativeIndexedObjectRegisterComponent|GetRegisteratIndex (Variables|Default|GetTeleportGroupObjectRegisterComp) _hiddentraveltargetgroup))
        (bind _register_objects (Utilities|Struct|BreakFortObjectArray _register))
        (if (and _bisvalid (Utilities|Array|IsNotEmpty _register_objects))
          (bind _output_get (LocalVariables|Setlocal_ObjectRegisterArray _register_objects))
          (Utilities|Array|RemoveItem _output_get self)
          (bind _attemptnorepeatdestinations (Verse|GetAttemptNoRepeatDestinationsVerse))
          (if (Math|Boolean|ANDBoolean _attemptnorepeatdestinations (Utilities|IsValid (Variables|Vars|GetPreviousTeleportTarget) :!class /Script/Engine.KismetSystemLibrary) (> (Utilities|Array|Length (LocalVariables|Getlocal_ObjectRegisterArray)) 1))
            (Utilities|Array|RemoveItem (LocalVariables|Getlocal_ObjectRegisterArray) (Variables|Vars|GetPreviousTeleportTarget)))
          (bind _element (ArrayMacros|GetRandomArrayElement (LocalVariables|Getlocal_ObjectRegisterArray))
            (:%20%20
              (bind _asb_hiding_prop (Utilities|Casting|CastToB_HidingProp _element)
                (:then
                  (bind _output_get_2 (Variables|Vars|SetPreviousTeleportTarget _asb_hiding_prop))
                  (CallFunction|UpdateTeleportTarget _output_get_2))
                (:CastFailed
                  _)))
            (:Empty%20Array
              _))
          (CallFunction|UpdateTeleportTarget 0)
          (else
            _))))))

(fn SetInvulnerability
  (params (Invulnerable :type Boolean))
  (if Invulnerable
    (Variables|Actor|SetbCanBeDamaged false) ; Invulnerable
    (Curie|DisableElementAttachmentonActor self "(TagName=\"Curie.Element.Fire\")")
    (Navigation|SetNavigationIndestructible true)
    (else
      (Variables|Actor|SetbCanBeDamaged true) ; Vulnerable
      (Curie|EnableElementAttachmentonActor self "(TagName=\"Curie.Element.Fire\")")
      (Navigation|SetNavigationIndestructible false)))
  (AI|RequestNavUpdateforBuilding self))

(fn SetAttemptNoRepeatDestinations_Verse
  (params (AttemptNoRepeatDestinations :type Boolean))
  (Utilities|Struct|SetmembersinHidingPropVerseOverrideValues (Variables|Verse|GetVerseOverrides) AttemptNoRepeatDestinations))

(fn SetBlockHideTime_Verse
  (params (BlockHideTime :type Float%20%28double-precision%29))
  (Utilities|Struct|SetmembersinHidingPropVerseOverrideValues (Variables|Verse|GetVerseOverrides) BlockHideTime))

(fn SetEnableHiddenTravel_Verse
  (params (EnableHiddenTravel :type Boolean))
  (Utilities|Struct|SetmembersinHidingPropVerseOverrideValues (Variables|Verse|GetVerseOverrides) EnableHiddenTravel)
  (if EnableHiddenTravel
    (CallFunction|EnableHiddenTravelEvent 0)
    (else
      (CallFunction|DisableHiddenTravelEvent 0))))

(fn SetHiddenTravelGroup_Override_Verse
  (params (HiddenTravelGroup_Override :type Boolean))
  (Utilities|Struct|SetmembersinHidingPropVerseOverrideValues (Variables|Verse|GetVerseOverrides) HiddenTravelGroup_Override))

(fn SetHiddenTravelGroup_Verse
  (params (HiddenTravelGroup :type Integer))
  (CallFunction|DisableHiddenTravelEvent 0)
  (Utilities|Struct|SetmembersinHidingPropVerseOverrideValues (Variables|Verse|GetVerseOverrides) HiddenTravelGroup)
  (CallFunction|EnableHiddenTravelEvent 0))

(fn SetHiddenTravelTargetGroup_Override_Verse
  (params (HiddenTravelTargetGroup_Override :type Boolean))
  (Utilities|Struct|SetmembersinHidingPropVerseOverrideValues (Variables|Verse|GetVerseOverrides) HiddenTravelTargetGroup_Override))

(fn SetHiddenTravelTargetGroup_Verse
  (params (HiddenTravelTargetGroup :type Integer))
  (CallFunction|DisableHiddenTravelEvent 0)
  (Utilities|Struct|SetmembersinHidingPropVerseOverrideValues (Variables|Verse|GetVerseOverrides) HiddenTravelTargetGroup)
  (CallFunction|EnableHiddenTravelEvent 0))

(fn SetMaxHidingTime_Override_Verse
  (params (Override :type Boolean))
  (Utilities|Struct|SetmembersinHidingPropVerseOverrideValues (Variables|Verse|GetVerseOverrides) Override))

(fn SetMaxHidingTime_Verse
  (params (MaxHidingTime :type Float%20%28double-precision%29))
  (Utilities|Struct|SetmembersinHidingPropVerseOverrideValues (Variables|Verse|GetVerseOverrides) MaxHidingTime))

(fn SetNumberOfOccupants_Verse
  (params (NewParam :type Integer))
  (Utilities|Struct|SetmembersinHidingPropVerseOverrideValues (Variables|Verse|GetVerseOverrides) NewParam))

(fn SetVerseOverridesToUserOptions
  (Utilities|Struct|SetmembersinHidingPropVerseOverrideValues (Variables|Verse|GetVerseOverrides) (Variables|UserOptions|GetNumberOfOccupants) (Variables|UserOptions|GetbWobbleWhileHiding) (Variables|UserOptions|GetSetMaxHidingTime) (Variables|UserOptions|GetBlockHideTime) (Variables|UserOptions|Advanced|GetHiddenTravelGroup) (Variables|UserOptions|Advanced|GetHiddenTravelTargetGroup) (Variables|UserOptions|Advanced|GetbAttemptNoRepeats) (Variables|Vars|GetHiddenTravelEnabledOnDevice) (Variables|OptionOverrides|GetSetMaxHidingTime_Override) (Variables|OptionOverrides|GetHiddenTravelGroup_Override) (Variables|OptionOverrides|GetHiddenTravelTargetGroup_Override)))

(fn SetWobbleWhileHiding_Verse
  (params (WobbleWhileHiding :type Boolean))
  (Utilities|Struct|SetmembersinHidingPropVerseOverrideValues (Variables|Verse|GetVerseOverrides) WobbleWhileHiding)
  (Variables|Default|SetbCanRustleAndWobble WobbleWhileHiding))

(fn ShouldBlockPlacement
  (params (MoveTool :type Fort%20Creative%20Move%20Tool%20Object%20Reference))
  (return false))

(fn ConstructionScript
  (|CallParentFunction))
