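; Creative port-a-potty hiding prop device with a selectable type that sets its body and door look.
; Option changes async-load meshes and textures; one body variant uses building texture data, so the Mesh variable is also set to prevent C++ resets.
; Entering activates an effect and swings the door via server-triggered timeline; occupant changes also swing it, except for teleporting pawns.
; Exposed asset provider overrides register all swappable assets as allowed references.
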(event OnPropertiesLoaded(ToyOptionsComponent)
  (CallFunction|UpdateMeshandMaterials))

(event OnAnyPropertyChanged(ToyOptionsComponent)
  (CallFunction|UpdateMeshandMaterials))

(event Custom|HidingPlayerCountChanged
  (|CallParentFunction)
  (bind _lastpawntohide (Variables|Default|GetLastPawnToHide)
    (:then
      (if (GameplayTags|HasMatchingGameplayTag _lastpawntohide (Variables|Default|GetTeleportingTag))
        _
        (elif (Networking|IsServer)
          (CallFunction|DoorSwingAnim))))
    (:else)))

(event EventInteractEnter
  (Components|Activation|Activate (Variables|Default|GetP_HidingProp_Portapotty) true)
  (if (Networking|IsServer)
    (CallFunction|DoorSwingAnim)))

(event Custom|DoorSwingAnim
  (bind (_direction _float) (Timeline|DoorSwing PlayFromStart)
    (:Update
      (Transformation|SetRelativeRotation (Variables|Default|GetConstr_Portapotty_Searched_Door) (Math|Rotator|Lerp(Rotator) "0, 0, 0" "0,140.000000, 0" _float) false false))))

(event Custom|UpdateMeshAndMaterials
  (Utilities|FlowControl|Sequence
    (:then_0
      ; Main Body Mesh
      (bind _loaded_object (Creative|Loading|AsyncLoadEditorSafe (Utilities|Map|Find (Variables|Default|GetBodyMeshMap) (Variables|UserOptions|Advanced|GetHidingPropMesh)))
        (:Loaded
          (bind _asstatic_mesh (Utilities|Casting|CastToStaticMesh _loaded_object))
          (Components|StaticMesh|SetStaticMesh (Variables|Debug|GetStaticMeshComponent) _asstatic_mesh)
          (Variables|Mesh|SetStaticMesh _asstatic_mesh) ; Need to set this or else the BuildingSMActor C++ \nwill reset the static mesh when setting the \nbuilding textures.
        )))
    (:then_1
      ; Door Mesh
      (bind _loaded_object_2 (Creative|Loading|AsyncLoadEditorSafe (Utilities|Map|Find (Variables|Default|GetDoorMeshMap) (Variables|UserOptions|Advanced|GetHidingPropMesh)))
        (:Loaded
          (bind _asstatic_mesh_2 (Utilities|Casting|CastToStaticMesh _loaded_object_2))
          (Components|StaticMesh|SetStaticMesh (Variables|Default|GetConstr_Portapotty_Searched_Door) _asstatic_mesh_2))))
    (:then_2
      ; Main Body Textures
      (switch Utilities|FlowControl|Switch|SwitchonEPortAPottyTypes (Variables|UserOptions|Advanced|GetHidingPropMesh)
        (:NewEnumerator0
          (bind _loaded_object_3 (Creative|Loading|AsyncLoadEditorSafe (Variables|Default|GetStandardBodyTexture))
            (:Loaded
              (bind _asmaterial_interface (Utilities|Casting|CastToMaterialInterface _loaded_object_3))
              (Rendering|Material|SetMaterial (Variables|Debug|GetStaticMeshComponent) 0 _asmaterial_interface))))
        (:NewEnumerator1
          (bind _loaded_object_4 (Creative|Loading|AsyncLoadEditorSafe (Variables|Default|GetMilitaryBodyBuildingTexture))
            (:Loaded
              (bind _asbuilding_texture_data (Utilities|Casting|CastToBuildingTextureData _loaded_object_4))
              (CallFunction|SetBuildingTextureData 0 _asbuilding_texture_data))))))
    (:then_3
      ; Door Texture
      (bind _loaded_object_5 (Creative|Loading|AsyncLoadEditorSafe (Utilities|Map|Find (Variables|Default|GetDoorTextureMap) (Variables|UserOptions|Advanced|GetHidingPropMesh)))
        (:Loaded
          (bind _asmaterial_interface_2 (Utilities|Casting|CastToMaterialInterface _loaded_object_5))
          (Rendering|Material|SetMaterial (Variables|Default|GetConstr_Portapotty_Searched_Door) 0 _asmaterial_interface_2))))))

(fn BuildExposedAssetProviderVersion(Implementation)
  (params (Version :type Fort%20Exposed%20Asset%20Provider%20Version%20Structure%20%28by%20ref%29))
  (|CallParentFunction Version)
  (ExposedAssetProvider|AddVersion Version 1)
  (return))

(fn CollectAssetReferencestoAllow(Implementation)
  (params (Collector :type Fort%20Exposed%20Asset%20Reference%20Collector%20Structure%20%28by%20ref%29))
  (|CallParentFunction Collector)
  (ExposedAssetProvider|CollectIndirectAssetReference (Variables|GetCollector) (Variables|Default|GetMilitaryBodyBuildingTexture))
  (ExposedAssetProvider|CollectIndirectAssetReference (Variables|GetCollector) (Variables|Default|GetStandardBodyTexture))
  (bind (_keypin _valuepin) (Utilities|Map|ForEachLoop(Map) (Variables|Default|GetBodyMeshMap))
    (:then
      (ExposedAssetProvider|CollectIndirectAssetReference (Variables|GetCollector) _valuepin))
    (:CompletedPin
      (bind (_keypin_2 _valuepin_2) (Utilities|Map|ForEachLoop(Map) (Variables|Default|GetDoorMeshMap))
        (:then
          (ExposedAssetProvider|CollectIndirectAssetReference (Variables|GetCollector) _valuepin_2))
        (:CompletedPin
          (bind (_keypin_3 _valuepin_3) (Utilities|Map|ForEachLoop(Map) (Variables|Default|GetDoorTextureMap))
            (:then
              (ExposedAssetProvider|CollectIndirectAssetReference (Variables|GetCollector) _valuepin_3))
            (:CompletedPin
              (return))))))))

(fn ConstructionScript
  (|CallParentFunction))
