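; Creative scarecrow hiding prop device that can rotate with its hidden player's camera and exposes pumpkin head and clothes color options.
; When occupancy changes, the authority starts or stops a timer that interpolates actor yaw toward the player's control rotation.
; It forces net updates to reduce lag, with an OnRep applying rotation on clients.
; Option changes toggle the head and set mapped clothes materials, loaded blocking in editor and async otherwise.
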(event EventBeginPlay
  (|CallParentFunction))

(event Collision|EventActorBeginOverlap
  (params (OtherActor :type Actor%20Object%20Reference))
  (|CallParentFunction OtherActor) ; This node is disabled and will not be called.\nDrag off pins to build functionality.
)

(event EventTick
  (params (DeltaSeconds :type Float%20%28single-precision%29))
  (|CallParentFunction DeltaSeconds) ; This node is disabled and will not be called.\nDrag off pins to build functionality.
)

(event EventHidingPlayerCountChanged
  (bind _hidingplayers (Variables|Default|GetHidingPlayers))
  ;[0.01807 0.5781 0.5781 0.569] This part is copied from B_HidingProp_Scarecrow for the 'follow camera' functionality
  (|CallParentFunction)
  (Networking|SwitchHasAuthority
    (:Authority
      (if (> (Utilities|Array|Length _hidingplayers) 0)
        (Variables|Default|SetbIsRotatingWithPlayer (ScalableFloat|GetValueasBool (Variables|Default|GetRotateWithPlayerBool)))
        (Variables|Default|SetPlayerReference (Utilities|Array|Get(acopy) _hidingplayers))
        (Variables|Default|SetCurrentRotation (Transformation|GetActorRotation))
        (bind _returnvalue (Utilities|Time|SetTimerbyEvent (delegate Custom|MoveWithCameraLogic) (ScalableFloat|GetValueatLevel (Variables|Default|GetTick%20Value)) true true))
        (Variables|Default|SetUpdateCameraTimer _returnvalue)
        (else
          (CallFunction|StopMoveCameraLogic))))))

(event Custom|MoveWithCameraLogic
  (bind _playerreference (Variables|Default|GetPlayerReference))
  (bind (_returnvalue_roll _returnvalue_pitch _returnvalue_yaw) (Math|Rotator|BreakRotator (Pawn|GetControlRotation _playerreference :!class /Script/Engine.Pawn)))
  (bind _interp_speed (Variables|Default|GetInterp%20Speed))
  (bind _returnvalue (ScalableFloat|GetValueatLevel _interp_speed))
  (bind _returnvalue_2 (Utilities|Time|GetWorldDeltaSeconds))
  (bind _targetrotation (Variables|Default|GetTargetRotation))
  (bind _currentrotation (Variables|Default|GetCurrentRotation))
  (bind _returnvalue_3 (Math|Interpolation|RInterpTo _currentrotation _targetrotation _returnvalue_2 _returnvalue))
  ;[0.01807 0.5781 0.5781 0.569] This part is copied from B_HidingProp_Scarecrow for the 'follow camera' functionality
  (if (Variables|Default|GetbIsRotatingWithPlayer)
    (Utilities|IsValid _playerreference
      (:Is%20Valid
        (Variables|Default|SetTargetRotation (Math|Rotator|MakeRotator 0.0 0.0 (+ _returnvalue_yaw -9e+01)))
        (Transformation|SetActorRotation _returnvalue_3)
        (Variables|Default|SetCurrentRotation _returnvalue_3)
        ;;[0.8542 0.2535 0.2025 0.5] Force Net replicate the actor's roatation immediately to client to avoid lagging issue
        (Networking|ForceNetUpdate))
      (:Is%20Not%20Valid
        _))
    (else
      _))
  (CallFunction|StopMoveCameraLogic))

(event Custom|Stop%20Move%20Camera%20Logic
  ;[0.01807 0.5781 0.5781 0.569] This part is copied from B_HidingProp_Scarecrow for the 'follow camera' functionality
  (Variables|Default|SetbIsRotatingWithPlayer false)
  (Variables|Default|SetPlayerReference 0)
  (Utilities|Time|ClearandInvalidateTimerbyHandle (Variables|Default|GetUpdateCameraTimer)))

(event OnAnyPropertyChanged(ToyOptionsComponent)
  (bind _pumpkinhead (Variables|Default|GetPumpkinHead))
  (bind _bhaspumpkinhead (Variables|UserOptions|Basic|GetbHasPumpkinHead))
  (Rendering|SetVisibility _pumpkinhead _bhaspumpkinhead false)
  (Development|SetHiddeninGame _pumpkinhead (not _bhaspumpkinhead) false)
  (CallFunction|SetClothesMaterial))

(event OnPropertiesLoaded(ToyOptionsComponent)
  (bind _pumpkinhead (Variables|Default|GetPumpkinHead))
  (bind _bhaspumpkinhead (Variables|UserOptions|Basic|GetbHasPumpkinHead))
  (Rendering|SetVisibility _pumpkinhead _bhaspumpkinhead false)
  (Development|SetHiddeninGame _pumpkinhead (not _bhaspumpkinhead) false)
  (CallFunction|SetClothesMaterial))

(event Custom|SetClothesMaterial
  (bind _clothescolor (Variables|UserOptions|Basic|GetClothesColor))
  (bind _clothesmaterialmap (Variables|Default|GetClothesMaterialMap))
  (bind (_value _returnvalue) (Utilities|Map|Find _clothesmaterialmap _clothescolor))
  (bind _staticmeshcomponent (Variables|Debug|GetStaticMeshComponent))
  (if _returnvalue
    ; Using the same logic in Async Load Editor Safe macro to async load material
    (if (Editor|InEditor)
      (bind _returnvalue_2 (Utilities|LoadAssetBlocking _value))
      (bind _asmaterial_instance (Utilities|Casting|CastToMaterialInstance _returnvalue_2))
      (Rendering|Material|SetMaterial _staticmeshcomponent 0 _asmaterial_instance)
      (else
        (bind _object (Utilities|AsyncLoadAsset _value)
          (:Completed
            (bind _asmaterial_instance_2 (Utilities|Casting|CastToMaterialInstance _object))
            (Rendering|Material|SetMaterial _staticmeshcomponent 0 _asmaterial_instance_2)))))))

(fn OnRep_CurrentRotation
  (Transformation|SetActorRotation (Variables|Default|GetCurrentRotation)))

(fn ConstructionScript
  (|CallParentFunction)
  ; Restore the clothes' material when loading in UEFN.
  (if (Editor|InEditor)
    (CallFunction|SetClothesMaterial)))
